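; Blueprint function library of shared static helpers for Fortnite missions.
; Spatial helpers cover player and pawn proximity, line-of-sight traces and vector math, while quest helpers query each controller's quest manager for completion and send stat updates.
; Others read difficulty curve tables, fetch objective badges, set stencil highlights, launch hoverboards and send survivor NPC gameplay events.
; An embedded note warns that HasPlayerCompletedQuest returns false when the player lacks the quest.
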(fn AnyPlayerPawnNearLocation
  (params (DistanceToCheck :type Float%20%28double-precision%29) (Location :type Vector) (Ignore%20Player%20Pawns :type Array%20of%20Fort%20Player%20Pawn%20Object%20References))
  (bind _outfortplayerpawns (Fortnite|GetFortPlayerPawns 0))
  (LocalVariables|SetFiltered%20Player%20Pawns _outfortplayerpawns)
  (for _array_element Ignore%20Player%20Pawns
    (Utilities|Array|RemoveItem (LocalVariables|GetFiltered%20Player%20Pawns) _array_element) ; remove ignored pawns
  )
  (for _array_element_2 (LocalVariables|GetFiltered%20Player%20Pawns)
    (bind _distance (CallFunction|DistanceBetweenTwoVectors (Transformation|GetActorLocation _array_element_2) (Variables|GetLocation)))
    (if (<= _distance DistanceToCheck)
      (bind _bdidactorhavetags (Class|AthenaQuestFunctionLibrary|AthenaCheckActorHasTags _array_element_2 "(GameplayTags=((TagName=\"Pawn.Athena.NonParticipant\")))" true)) ; Pawn.Athena.NonParticipant - looks for any non-players\n..NPCs\n..Enemy AI like Henchmen, IO Guards, Bosses
      (if (not _bdidactorhavetags)
        (Utilities|Array|AddUnique (LocalVariables|GetPawns)) ; this player is nearby
      )))
  (return (>= (Utilities|Array|Length (LocalVariables|GetPawns)) 1) (LocalVariables|GetPawns)))

(fn AnyPawnNearActor
  (params (DistanceToCheck :type Float%20%28double-precision%29) (Actor :type Actor%20Object%20Reference))
  (Utilities|IsValid Actor
    (:Is%20Valid
      (bind _outactors (Actor|GetAllActorsOfClass 0 "/Script/FortniteGame.FortPawn"))
      (for _array_element _outactors
        (if (<= (Transformation|GetDistanceTo Actor _array_element) DistanceToCheck)
          (Utilities|Array|AddUnique (LocalVariables|GetPawns) _array_element)))
      (return (>= (Utilities|Array|Length (LocalVariables|GetPawns)) 1) (LocalVariables|GetPawns)))
    (:Is%20Not%20Valid
      (Development|PrintString 0 "Pawn Near Actor - Actor isnt valid "))))

(fn AnyPlayerNearActor
  (params (DistanceToCheck :type Float%20%28double-precision%29) (Actor :type Actor%20Object%20Reference))
  (Utilities|IsValid Actor
    (:Is%20Valid
      (bind _outfortplayerpawns (Fortnite|GetFortPlayerPawns 0))
      (for _array_element _outfortplayerpawns
        (if (<= (Transformation|GetDistanceTo Actor _array_element) DistanceToCheck)
          (Utilities|Array|AddUnique (LocalVariables|GetPlayerPawns) _array_element)))
      (return (>= (Utilities|Array|Length (LocalVariables|GetPlayerPawns)) 1) (LocalVariables|GetPlayerPawns)))
    (:Is%20Not%20Valid
      (Development|PrintString 0 "Player Near Actor - Actor isnt valid "))))

(fn ApplyDifficultyOffset
  (params (RowName :type Name) (BaseDifficulty :type Float%20%28double-precision%29))
  (bind _difficulty (CallFunction|GetCurrentDifficulty (Variables|Get__WorldContext)))
  (bind _outxy (DataTable|EvaluateCurveTableRow "/Game/Balance/DataTables/DifficultyOffsets.DifficultyOffsets" RowName BaseDifficulty)
    (:RowFound
      (MissionHelpers|AdjustWaveBasedGameDifficulty (+ _outxy (- BaseDifficulty _difficulty)))
      (return true))
    (:RowNotFound
      (return false))))

(fn CheckBasicActorLineOfSight
  (params (Actor%20A :type Actor%20Object%20Reference) (ActorA%20Location%20Offset :type Vector) (Actor%20B :type Actor%20Object%20Reference) (Actor%20B%20Location%20Offset :type Vector) (UseActorHalfHeightForSight :type Boolean) (TraceChannel :type ETraceTypeQuery%20Enum :default "TraceTypeQuery1") (DrawDebugType :type EDrawDebugTrace%20Enum) (DrawTime :type Float%20%28double-precision%29) (bTraceComplex :type Boolean) (bIgnoreSelf :type Boolean :default true) (Ignored%20Actors :type Array%20of%20Actor%20Object%20References))
  (bind (_origin _boxextent) (Collision|GetActorBounds (Variables|GetActor%20B) true false))
  (bind (_origin_2 _boxextent_2) (Collision|GetActorBounds (Variables|GetActor%20A) true false))
  (Utilities|IsValid (Variables|GetActor%20A)
    (:Is%20Valid
      (Utilities|IsValid (Variables|GetActor%20B)
        (:Is%20Valid
          (LocalVariables|SetActorA_StartVector (+ (Transformation|GetActorLocation (Variables|GetActor%20A)) (Variables|GetActorA%20Location%20Offset)))
          (LocalVariables|SetActor%20B%20Start%20Vector (+ (Transformation|GetActorLocation (Variables|GetActor%20B)) (Variables|GetActor%20B%20Location%20Offset)))
          (if (Variables|GetUseActorHalfHeightForSight)
            (bind _returnvalue (GameplayTags|HasMatchingGameplayTag(Message) (Variables|GetActor%20A) (LocalVariables|GetTag_IgnoreHalfHeightForSightCheck)))
            (if (not _returnvalue)
              (LocalVariables|SetActorA_StartVector (+ (LocalVariables|GetActorA_StartVector) (Math|Vector|MakeVector 0.0 0.0 (/ (Math|Float|Clamp(Float) (.z _boxextent_2) 1.0 99999.0) 2.0)))))
            (bind _returnvalue_2 (GameplayTags|HasMatchingGameplayTag(Message) (Variables|GetActor%20B) (LocalVariables|GetTag_IgnoreHalfHeightForSightCheck)))
            (if (not _returnvalue_2)
              (LocalVariables|SetActor%20B%20Start%20Vector (+ (LocalVariables|GetActor%20B%20Start%20Vector) (Math|Vector|MakeVector 0.0 0.0 (/ (Math|Float|Clamp(Float) (.z _boxextent) 1.0 99999.0) 2.0))))))
          (bind _outhit (Collision|LineTraceByChannel 0 (LocalVariables|GetActorA_StartVector) (LocalVariables|GetActor%20B%20Start%20Vector) (Variables|GetTraceChannel) (Variables|GetbTraceComplex) (Variables|GetIgnored%20Actors) (Variables|GetDrawDebugType) (Variables|GetbIgnoreSelf) "(R=1.000000,G=0.000000,B=0.000000,A=1.000000)" "(R=0.000000,G=1.000000,B=0.000000,A=1.000000)" (Variables|GetDrawTime)))
          (if (Collision|BreakHitResult _outhit)
            (if (Utilities|Equal(Object) (.HitActor (Collision|BreakHitResult _outhit)) (Variables|GetActor%20B))
              (LocalVariables|SetResult true)
              (else
                (LocalVariables|SetResult false)))
            (return (LocalVariables|GetResult) (.Distance (Collision|BreakHitResult _outhit)))
            (else
              (return false 0.0))))
        (:Is%20Not%20Valid
          (return false 0.0))))
    (:Is%20Not%20Valid
      (return false 0.0))))

(fn CheckBasicVectorToActorLineOfSight
  (params (Vector%20A :type Vector) (Actor%20B :type Actor%20Object%20Reference) (Actor%20B%20Location%20Offset :type Vector) (TraceChannel :type ETraceTypeQuery%20Enum :default "TraceTypeQuery1") (DrawDebugType :type EDrawDebugTrace%20Enum) (DrawTime :type Float%20%28double-precision%29) (ActorsToIgnore :type Array%20of%20Actor%20Object%20References) (bTraceComplex :type Boolean) (bRequireThatTraceHitActorB :type Boolean :default true))
  (bind _returnvalue (Transformation|GetActorLocation Actor%20B))
  (bind _returnvalue_2 (+ Actor%20B%20Location%20Offset _returnvalue))
  (Utilities|IsValid Actor%20B
    (:Is%20Valid
      (bind _outhit (Collision|LineTraceByChannel 0 Vector%20A _returnvalue_2 TraceChannel bTraceComplex ActorsToIgnore DrawDebugType true "(R=1.000000,G=0.000000,B=0.000000,A=1.000000)" "(R=0.000000,G=1.000000,B=0.000000,A=1.000000)" DrawTime))
      (if (Collision|BreakHitResult _outhit)
        (if (Utilities|Equal(Object) Actor%20B (.HitActor (Collision|BreakHitResult _outhit)))
          (LocalVariables|SetResult true)
          (else
            (LocalVariables|SetResult false)))
        (return (LocalVariables|GetResult) (.Distance (Collision|BreakHitResult _outhit)))
        (elif (Variables|GetbRequireThatTraceHitActorB)
          (return false 0.0)
          (else
            (return true (Math|Vector|Distance(Vector) Vector%20A _returnvalue_2))))))))

(fn DistanceBetweenTwoVectors
  (params (vector1 :type Vector) (Vector2 :type Vector))
  (return (Math|Vector|VectorLength (- vector1 Vector2))))

(fn DoesAnyoneRequireQuest
  (params (QuestReference :type Fort%20Quest%20Item%20Definition%20Object%20Reference) (Quest%20Backend%20Name :type Name))
  (bind _returnvalue (Fortnite|GetAllFortPlayerControllers true))
  (if (>= (Utilities|Array|Length _returnvalue) 1)
    (bind _array_element (Utilities|Array|ForEachLoopwithBreak _returnvalue :!label _label)
      (:LoopBody
        (if (Utilities|IsValid (Quests|GetQuestwithDefinition (Quests|GetQuestManager _array_element "Invalid") QuestReference) :!class /Script/Engine.KismetSystemLibrary)
          (if (not (Quest|HasCompletedObjectivewithName (Quests|GetQuestManager _array_element "Invalid") QuestReference Quest%20Backend%20Name :!class /Script/FortniteGame.FortQuestManager))
            (LocalVariables|SetPlayerNeedsQuest true)
            (goto _label Break))))
      (:Completed
        _))
    (else
      (LocalVariables|SetPlayerNeedsQuest false)))
  (return (LocalVariables|GetPlayerNeedsQuest)))

(fn DoesAnyPlayerRequireQuestObjective
  (params (QuestReference :type Fort%20Quest%20Item%20Definition%20Object%20Reference) (QuestObjectiveHandle :type Data%20Table%20Row%20Handle%20Structure :default "(DataTable=None,RowName=\"\")"))
  (bind _returnvalue (Fortnite|GetAllFortPlayerControllers true))
  (Utilities|IsValid QuestReference
    (:Is%20Valid
      (if (>= (Utilities|Array|Length _returnvalue) 1)
        (for _array_element _returnvalue
          (Utilities|IsValid _array_element
            (:Is%20Valid
              (Utilities|IsValid (Quests|GetQuestwithDefinition (Quests|GetQuestManager _array_element "Invalid") QuestReference)
                (:Is%20Valid
                  (if (not (Quest|HasCompletedObjective (Quests|GetQuestwithDefinition (Quests|GetQuestManager _array_element "Invalid") QuestReference) QuestObjectiveHandle :!class /Script/FortniteGame.FortQuestItem))
                    (Utilities|Array|AddUnique (LocalVariables|GetControllersWhoRequireObjective) _array_element)
                    (LocalVariables|SetSomebodyNeededTheObjective true)))))))
        (return (LocalVariables|GetControllersWhoRequireObjective) (LocalVariables|GetSomebodyNeededTheObjective))))))

(fn DoesPlayerHaveQuest
  (params (PlayerController :type Fort%20Player%20Controller%20Object%20Reference) (Quest%20Item%20Reference :type Fort%20Quest%20Item%20Definition%20Object%20Reference) (Quest%20Item%20Objective%20Backend%20Name :type Name))
  (bind _returnvalue (Quests|GetQuestManager PlayerController "Invalid"))
  (bind _returnvalue_2 (Quests|GetQuestwithDefinition _returnvalue Quest%20Item%20Reference))
  (Utilities|IsValid PlayerController
    (:Is%20Valid
      (Utilities|IsValid Quest%20Item%20Reference
        (:Is%20Valid
          (if (Utilities|IsValid _returnvalue_2 :!class /Script/Engine.KismetSystemLibrary)
            (if (not (Quest|HasCompletedObjectivewithName _returnvalue_2 Quest%20Item%20Objective%20Backend%20Name :!class /Script/FortniteGame.FortQuestItem))
              (return true PlayerController)))
          (return false PlayerController))))))

(fn DropChance
  (params (DropPercentChance :type Float%20%28double-precision%29))
  (if (>= DropPercentChance 100)
    (return true)
    (elif (<= (Math|Random|RandomFloatinRange 0.0 100) (Math|Float|Clamp(Float) DropPercentChance 0.0 100))
      (return true)
      (else
        (return false)))))

(fn FlashObjectiveUI
  (params (Objective%20Reference :type Fort%20Objective%20Base%20Object%20Reference))
  (Mission|SendMissionUIEvent Objective%20Reference (GameplayTags|MakeLiteralGameplayTagContainer "(GameplayTags=((TagName=\"Activities.UIEvent.DrawAttention\")))")))

(fn GetBuildingRecommendation
  (params (RowName :type Name))
  (bind (_difficulty _success) (CallFunction|GetCurrentDifficulty (Variables|Get__WorldContext)))
  (if _success
    (bind _outxy (DataTable|EvaluateCurveTableRow "/Game/Balance/DataTables/BuildingMissionRecs.BuildingMissionRecs" RowName _difficulty)
      (:RowFound
        (return (Math|Float|Round _outxy) true))
      (:RowNotFound
        _)))
  (return 0 false))

(fn GetContributingControllersNearActor
  (params (Distance :type Float%20%28double-precision%29) (Actor :type Actor%20Object%20Reference))
  (Utilities|IsValid Actor
    (:Is%20Valid
      (bind (_isplayernearactor _playersnearactor) (CallFunction|AnyPlayerNearActor Distance Actor))
      (if _isplayernearactor
        (for _array_element _playersnearactor
          (bind _asfort_player_controller (Utilities|Casting|CastToFortPlayerController (Pawn|GetController _array_element)))
          (Utilities|Array|AddUnique (LocalVariables|GetControllersNearbyArray) _asfort_player_controller))
        (LocalVariables|SetFoundPlayer true)
        (else
          (LocalVariables|SetFoundPlayer false)))
      (return (LocalVariables|GetControllersNearbyArray) (LocalVariables|GetFoundPlayer)))
    (:Is%20Not%20Valid
      (Development|PrintString 0 "Contributing Controller Failed, the actor was not valid so we can't compare distance"))))

(fn GetControllersNearActorThatRequireQuestObjective
  (params (Distance :type Float%20%28double-precision%29) (Actor :type Actor%20Object%20Reference) (Quest%20Item :type Fort%20Quest%20Item%20Definition%20Object%20Reference) (Quest%20Item%20Backend%20Objective%20Name :type Name))
  (Utilities|IsValid Actor
    (:Is%20Valid
      (Utilities|IsValid Quest%20Item
        (:Is%20Valid
          (LocalVariables|SetLocalQuestItem Quest%20Item)
          (LocalVariables|SetLocalBackendName Quest%20Item%20Backend%20Objective%20Name)
          (bind (_isplayernearactor _playersnearactor) (CallFunction|AnyPlayerNearActor Distance Actor))
          (if _isplayernearactor
            (for _array_element _playersnearactor
              (bind _asfort_player_controller (Utilities|Casting|CastToFortPlayerController (Pawn|GetController _array_element)))
              (Utilities|Array|AddUnique (LocalVariables|GetControllersNearbyArray) _asfort_player_controller))
            (for _array_element_2 (LocalVariables|GetControllersNearbyArray)
              (bind (_questvalid _playercontrollerref) (Quest|DoesPlayerHaveQuest _array_element_2 (LocalVariables|GetLocalQuestItem) (LocalVariables|GetLocalBackendName)))
              (if _questvalid
                (LocalVariables|SetFoundPlayer true)
                (Utilities|Array|AddUnique (LocalVariables|GetControllersWhoRequireQuestObjective) _playercontrollerref)))
            (else
              (LocalVariables|SetFoundPlayer false :!label _label)))
          (return (LocalVariables|GetControllersWhoRequireQuestObjective) (LocalVariables|GetFoundPlayer)))
        (:Is%20Not%20Valid
          _)))
    (:Is%20Not%20Valid
      _))
  (goto _label))

(fn GetCurrentDifficulty
  (bind _asfort_game_state_zone (Utilities|Casting|CastToFortGameStateZone (Game|GetGameState (Variables|Get__WorldContext)))
    (:then
      (return (Difficulty|GetGameDifficulty _asfort_game_state_zone) true))
    (:CastFailed
      (return 0.0 false))))

(fn GetObjectiveBadgeIconBrush
  (params (Objective :type Fort%20Objective%20Base%20Object%20Reference) (RewardTag :type Gameplay%20Tag%20Structure))
  (bind _badgeitem (Default|GetObjectiveBadgeItem Objective RewardTag))
  (Utilities|IsValid _badgeitem
    (:Is%20Valid
      (return (Item|GetSmallPreviewImageBrush _badgeitem)))
    (:Is%20Not%20Valid
      (return 0))))

(fn GetObjectiveBadgeItem
  (params (Objective :type Fort%20Objective%20Base%20Object%20Reference) (RewardTag :type Gameplay%20Tag%20Structure))
  (bind _outrewarditems (Missions|GetRewardItemsbyTag Objective RewardTag :!class /Script/FortniteGame.FortObjectiveBase))
  (Math|Integer|CompareInt (Utilities|Array|Length _outrewarditems)
    (:>
      (bind _array_element (Utilities|Array|ForEachLoopwithBreak _outrewarditems)
        (:LoopBody
          (bind _asfort_badge_item_definition (Utilities|Casting|CastToFortBadgeItemDefinition _array_element))
          (return _asfort_badge_item_definition))
        (:Completed
          (return 0))))))

(fn GetObjectiveBadgeSoftIcon
  (params (Objective :type Fort%20Objective%20Base%20Object%20Reference) (RewardTag :type Gameplay%20Tag%20Structure))
  (bind _badgeitem (Default|GetObjectiveBadgeItem Objective RewardTag))
  (Utilities|IsValid _badgeitem
    (:Is%20Valid
      (return (Class|FortItemFunctionLibrary|GetSmallPreviewImage _badgeitem)))
    (:Is%20Not%20Valid
      (return ""))))

(fn GetSeasonalEventManager
  (params (WorldContextObject :type Object%20Reference))
  (bind _asfort_game_instance (Utilities|Casting|CastToFortGameInstance (Game|GetGameInstance WorldContextObject)))
  (return (Class|FortGameInstance|GetEventManager _asfort_game_instance)))

(fn GiveQuestUpdateToPlayers
  (params (Quest%20Reference :type Fort%20Quest%20Item%20Definition%20Object%20Reference) (Quest%20Objective%20Backend%20Name :type Name) (ObjectiveStatEvent :type Data%20Table%20Row%20Handle%20Structure) (PlayerControllersForUpdate :type Array%20of%20Fort%20Player%20Controller%20Object%20References))
  (for _array_element PlayerControllersForUpdate
    (Utilities|IsValid _array_element
      (:Is%20Valid
        (Utilities|IsValid (Quests|GetQuestwithDefinition (Quests|GetQuestManager _array_element "Invalid") Quest%20Reference)
          (:Is%20Valid
            (if (not (Quest|HasCompletedObjectivewithName (Quests|GetQuestwithDefinition (Quests|GetQuestManager _array_element "Invalid") Quest%20Reference) Quest%20Objective%20Backend%20Name :!class /Script/FortniteGame.FortQuestItem))
              (Utilities|IsValid (Utilities|Struct|BreakDataTableRowHandle ObjectiveStatEvent)
                (:Is%20Valid
                  (Quest|SendCustomStatEvent (Quests|GetQuestManager _array_element "Invalid") ObjectiveStatEvent 1))
                (:Is%20Not%20Valid
                  (Quest|SendCustomStatEventDirect (Quests|GetQuestManager _array_element "Invalid") Quest%20Objective%20Backend%20Name (Quests|GetQuestwithDefinition (Quests|GetQuestManager _array_element "Invalid") Quest%20Reference) 1))))))))))

(fn HasPlayerCompletedQuest
  (params (PlayerController :type Fort%20Player%20Controller%20Object%20Reference) (QuestReference :type Fort%20Quest%20Item%20Definition%20Object%20Reference) (QuestBackendName :type Name :default "None"))
  (bind _playercontroller (Variables|GetPlayerController))
  (bind _playerstate (Variables|/Script/Engine.Controller|GetPlayerState _playercontroller))
  (bind _returnvalue (Actor|GetComponentbyClass _playerstate "/Script/FortniteGame.FortPlayerStateComponent_Quests"))
  (bind _playercontroller_2 (Variables|GetPlayerController))
  (bind _returnvalue_2 (Quests|GetQuestManager _playercontroller_2 "Invalid"))
  (bind _returnvalue_3 (Quests|GetQuestwithDefinition _returnvalue_2 QuestReference))
  (Utilities|IsValid PlayerController
    (:Is%20Valid
      (Utilities|IsValid QuestReference
        (:Is%20Valid
          (if (Class|FortQuestItemDefinition|HasComponentforWorldProcessingUpdates QuestReference)
            (Utilities|IsValid _playerstate
              (:Is%20Valid
                (Utilities|IsValid _returnvalue
                  (:Is%20Valid
                    (bind _returnvalue_4 (Class|QuestInstanceManagerInterface|FindQuestInstanceforManager(Message) _returnvalue (Item|GetPersistentName (Variables|GetQuestReference))))
                    (if (Utilities|IsValidInterface _returnvalue_4)
                      (if (Utilities|Name|Equal(Name) (Variables|GetQuestBackendName) "None")
                        (bind _asfort_quest_item (Utilities|Casting|CastToFortQuestItem _returnvalue_4)
                          (:then
                            (return (Variables|GetPlayerController) (Quest|HasCompletedQuest _asfort_quest_item :!class /Script/FortniteGame.FortQuestItem)))
                          (:CastFailed
                            _)))
                      (return (Variables|GetPlayerController) (Class|QuestInstanceInterface|HasCompletedObjectiveInstance _returnvalue_4 (Variables|GetQuestBackendName)))
                      (else
                        (return (Variables|GetPlayerController) false))))
                  (:Is%20Not%20Valid
                    _)))
              (:Is%20Not%20Valid
                _))
            (return (Variables|GetPlayerController) false)
            (elif (Utilities|Name|Equal(Name) (Variables|GetQuestBackendName) "None")
              (Utilities|IsValid _returnvalue_3
                (:Is%20Valid
                  (LocalVariables|SetHasCompletedQuest (Quest|HasCompletedQuest _returnvalue_3 :!class /Script/FortniteGame.FortQuestItem))
                  (return (Variables|GetPlayerController) (LocalVariables|GetHasCompletedQuest)))
                (:Is%20Not%20Valid
                  _))
              (else
                (Utilities|IsValid _returnvalue_3
                  (:Is%20Valid
                    (LocalVariables|SetHasCompletedQuest (Quest|HasCompletedObjectivewithName _returnvalue_3 (Variables|GetQuestBackendName) :!class /Script/FortniteGame.FortQuestItem))
                    (return (Variables|GetPlayerController) (LocalVariables|GetHasCompletedQuest)))
                  (:Is%20Not%20Valid
                    _)))))
          ;[1 1 1] This function is not implemented in the wisest way. Since it needs to be used in conjunction with knowing if the player HAS the quest. Since asking a players quest manager GetQuestWithDefinition() can be a nullptr if they don't HAVE the quest. Meaning HasCompletedQuest() will always be false. Designers could potentially (and i'm sure already have) use this function without understanding there is a distinction.
          ;
          ; I saw a bunch of errors in the message log, so I added some IsValid checks hoping they'll go away. -Lacelle
          (return (Variables|GetPlayerController) false))
        (:Is%20Not%20Valid
          _)))
    (:Is%20Not%20Valid
      _))
  (return (Variables|GetPlayerController) false))

(fn HasPlayerCompletedQuestObjective
  (params (PlayerController :type Fort%20Player%20Controller%20Object%20Reference) (QuestReference :type Fort%20Quest%20Item%20Definition%20Object%20Reference) (QuestBackendObjectiveName :type Name :default "None"))
  (bind _returnvalue (Quests|GetQuestManager PlayerController "Invalid"))
  (bind _returnvalue_2 (Quests|GetQuestwithDefinition _returnvalue QuestReference))
  (Utilities|IsValid PlayerController
    (:Is%20Valid
      (Utilities|IsValid QuestReference
        (:Is%20Valid
          (if (Utilities|Name|Equal(Name) QuestBackendObjectiveName "None")
            (Utilities|IsValid _returnvalue_2
              (:Is%20Valid
                (LocalVariables|SetHasCompletedObjective true))
              (:Is%20Not%20Valid
                (LocalVariables|SetHasCompletedObjective false :!label _label)))
            (return PlayerController (LocalVariables|GetHasCompletedObjective))
            (else
              (Utilities|IsValid _returnvalue_2
                (:Is%20Valid
                  (LocalVariables|SetHasCompletedObjective (Quest|HasCompletedObjectivewithName _returnvalue_2 QuestBackendObjectiveName :!class /Script/FortniteGame.FortQuestItem))
                  (return PlayerController (LocalVariables|GetHasCompletedObjective)))
                (:Is%20Not%20Valid
                  _))))
          (goto _label))))))

(fn HasPlayerCompletedQuestObjectiveHandle
  (params (PlayerController :type Fort%20Player%20Controller%20Object%20Reference) (QuestReference :type Fort%20Quest%20Item%20Definition%20Object%20Reference) (QuestBackendObjectiveHandle :type Data%20Table%20Row%20Handle%20Structure :default "(DataTable=None,RowName=\"\")"))
  (bind _returnvalue (Quests|GetQuestManager PlayerController "Invalid"))
  (bind _returnvalue_2 (Quests|GetQuestwithDefinition _returnvalue QuestReference))
  (Utilities|IsValid PlayerController
    (:Is%20Valid
      (Utilities|IsValid QuestReference
        (:Is%20Valid
          (Utilities|IsValid _returnvalue_2
            (:Is%20Valid
              (if (Quest|HasCompletedObjective _returnvalue_2 QuestBackendObjectiveHandle :!class /Script/FortniteGame.FortQuestItem)
                (LocalVariables|SetHasCompletedObjective true)
                (else
                  (LocalVariables|SetHasCompletedObjective false))))
            (:Is%20Not%20Valid
              (LocalVariables|SetHasCompletedObjective false :!label _label)))
          (return PlayerController (LocalVariables|GetHasCompletedObjective)))
        (:Is%20Not%20Valid
          _)))
    (:Is%20Not%20Valid
      _))
  (goto _label))

(fn HaveAllPlayersCompletedQuest
  (params (QuestItem :type Fort%20Quest%20Item%20Definition%20Object%20Reference))
  (Utilities|IsValid QuestItem
    (:Is%20Valid
      (bind _array_element (Utilities|Array|ForEachLoopwithBreak (Fortnite|GetAllFortPlayerControllers true) :!label _label)
        (:LoopBody
          (Utilities|IsValid (Quests|GetQuestwithDefinition (Quests|GetQuestManager _array_element "Invalid") QuestItem)
            (:Is%20Valid
              (if (not (Quest|HasCompletedQuest (Quests|GetQuestwithDefinition (Quests|GetQuestManager _array_element "Invalid") QuestItem) :!class /Script/FortniteGame.FortQuestItem))
                (LocalVariables|SetNotCompleted true)
                (goto _label Break)))
            (:Is%20Not%20Valid
              (LocalVariables|SetNotCompleted false))))
        (:Completed
          _)))
    (:Is%20Not%20Valid
      (LocalVariables|SetNotCompleted true)))
  (return (LocalVariables|GetNotCompleted)))

(fn HideParticleComponentsAttachedToMesh
  (params (MeshComponent :type Static%20Mesh%20Component%20Object%20Reference))
  (bind _children (Components|GetChildrenComponents MeshComponent true))
  (Utilities|IsValid MeshComponent
    (:Is%20Valid
      (if (>= (Utilities|Array|Length _children) 1)
        (for _array_element _children
          (bind _ascascade_particle_system_component (Utilities|Casting|CastToParticleSystemComponent _array_element))
          (Utilities|IsValid _ascascade_particle_system_component
            (:Is%20Valid
              (Components|Activation|Deactivate _ascascade_particle_system_component)
              (LocalVariables|SetFoundParticle true))))
        (Utilities|Array|Clear _children)))
    (:Is%20Not%20Valid
      _))
  (return (LocalVariables|GetFoundParticle)))

(fn HighlightQuestActor
  (params (ActorToHighlight :type Actor%20Object%20Reference) (HighlightEnabled :type Boolean))
  (bind _asbuilding_actor (Utilities|Casting|CastToBuildingActor ActorToHighlight))
  (Class|BuildingActor|SetQuestHighlight _asbuilding_actor HighlightEnabled)
  (bind _asbuilding_static_mesh (Utilities|Casting|CastToBuildingSMActor _asbuilding_actor))
  (if HighlightEnabled
    (bind _returnvalue (Rendering|Material|GetScalarParameterValue 0 "/Game/Packages/Fortress_SharedMaterials/Stencil_ID_Definitions/Stencil_ID_Definitions.Stencil_ID_Definitions" "Green_OutlineID_CodeReferenced" :!class /Script/Engine.KismetMaterialLibrary))
    (Rendering|SetRenderCustomDepth (Building|GetBuildingMeshComponent _asbuilding_static_mesh) true)
    (Rendering|SetCustomDepthStencilValue (Building|GetBuildingMeshComponent _asbuilding_static_mesh) (Math|Float|Truncate _returnvalue))
    (else
      (bind _returnvalue_2 (Rendering|Material|GetScalarParameterValue 0 "/Game/Packages/Fortress_SharedMaterials/Stencil_ID_Definitions/Stencil_ID_Definitions.Stencil_ID_Definitions" "No Effect" :!class /Script/Engine.KismetMaterialLibrary))
      (Rendering|SetRenderCustomDepth (Building|GetBuildingMeshComponent _asbuilding_static_mesh) true)
      (Rendering|SetCustomDepthStencilValue (Building|GetBuildingMeshComponent _asbuilding_static_mesh) (Math|Float|Truncate _returnvalue_2)))))

(fn LaunchHoverboard
  (params (HoverboardActor :type Actor%20Object%20Reference) (LaunchVelocity :type Vector) (ForceFeedbackTag :type Name))
  (bind _hoverboard (LocalVariables|GetHoverboard))
  (bind _returnvalue (Vehicle|GetDriver _hoverboard))
  (bind _asfort_athena_jackal_vehicle (Utilities|Casting|CastToFortAthenaJackalVehicle HoverboardActor)
    (:then
      (LocalVariables|SetHoverboard _asfort_athena_jackal_vehicle)
      (if (not (GameplayTags|HasMatchingGameplayTag _asfort_athena_jackal_vehicle "(TagName=\"Gameplay.Action.Player.Launched\")"))
        (Audio|BroadcastAbilityImpactatLocation (LocalVariables|GetHoverboard) 0 (LocalVariables|GetHoverboard) LaunchVelocity 777.0)
        (Environment|ApplyGlobalEnvironmentGameplayEffecttoActor (LocalVariables|GetHoverboard) "/Game/Abilities/Player/Generic/GameplayEffects/GE_Generic_Character_Launch.GE_Generic_Character_Launch_C")
        (Vehicle|SetIgnoreNextFallingDamage (LocalVariables|GetHoverboard) true)
        (if (Utilities|IsValid _returnvalue :!class /Script/Engine.KismetSystemLibrary)
          (bind _asfort_player_controller (Utilities|Casting|CastToFortPlayerController (Pawn|GetController _returnvalue))
            (:then
              (Game|Feedback|LocalPlayForceFeedback _asfort_player_controller "/Game/Weapons/ForceFeedback/FF_Generic_LaunchPawn.FF_Generic_LaunchPawn" false ForceFeedbackTag))
            (:CastFailed
              _)))
        (return true)))
    (:CastFailed
      _))
  (return false))

(fn ListPlayersWhoNeedQuest
  (params (Quest%20Reference :type Fort%20Quest%20Item%20Definition%20Object%20Reference) (Quest%20Backend%20Name :type Name))
  (bind _returnvalue (Fortnite|GetAllFortPlayerControllers true))
  (bind _playerswhorequirequest (LocalVariables|GetPlayersWhoRequireQuest))
  (if (>= (Utilities|Array|Length _returnvalue) 1)
    (for _array_element _returnvalue
      (if (not (Quest|HasCompletedObjectivewithName (Quests|GetQuestManager _array_element "Invalid") Quest%20Reference Quest%20Backend%20Name :!class /Script/FortniteGame.FortQuestManager))
        (Utilities|Array|AddUnique (LocalVariables|GetPlayersWhoRequireQuest) _array_element))))
  (if (>= (Utilities|Array|Length _playerswhorequirequest) 1)
    (LocalVariables|SetPlayerNeededTheQuest true))
  (return _playerswhorequirequest (LocalVariables|GetPlayerNeededTheQuest)))

(fn NPC_CanBeElemental
  (params (PawnTag :type Gameplay%20Tag%20Structure))
  (return (GameplayTags|HasTag (LocalVariables|GetElementalEnablePawnTag) PawnTag true)))

(fn NPC_ConvertTagToNameText
  (params (PawnTag :type Gameplay%20Tag%20Structure))
  (switch Utilities|FlowControl|Switch|SwitchonGameplayTag PawnTag
    (:Default
      (return ""))
    (:NPC.CharacterType.Blaster.Basic
      (return "NSLOCTEXT(\"[E5389DCD8C585B8C00B0C668CB1CF226]\", \"11DD9D1B412B0AEB8E289B9167AED0D3\", \"Blaster\")"))
    (:NPC.CharacterType.Flinger.Basic
      (return "NSLOCTEXT(\"[E5389DCD8C585B8C00B0C668CB1CF226]\", \"8CAA17E94E606743428154A520735D1B\", \"Flinger\")"))
    (:NPC.CharacterType.Husk.Basic
      (return "NSLOCTEXT(\"[E5389DCD8C585B8C00B0C668CB1CF226]\", \"4F57F73849AFCF15C688178E5DE3E210\", \"Husk\")"))
    (:NPC.CharacterType.Husk.Beehive
      (return "NSLOCTEXT(\"[E5389DCD8C585B8C00B0C668CB1CF226]\", \"9ADA40054BB67C316A9C5FB87374DA2D\", \"Beehive Husk\")"))
    (:NPC.CharacterType.Husk.Dwarf
      (return "NSLOCTEXT(\"[E5389DCD8C585B8C00B0C668CB1CF226]\", \"DC69C20440D4DFB68D7643997976E976\", \"Huskling\")"))
    (:NPC.CharacterType.Husk.Husky
      (return "NSLOCTEXT(\"[E5389DCD8C585B8C00B0C668CB1CF226]\", \"54390CAD474DA81DDECF52910A65EF90\", \"Husky Husk\")"))
    (:NPC.CharacterType.Husk.Pitcher
      (return "NSLOCTEXT(\"[E5389DCD8C585B8C00B0C668CB1CF226]\", \"ECF3201E4653E775E6005BA22D455724\", \"Pitcher\")"))
    (:NPC.CharacterType.Husk.PoisonLobber
      (return "NSLOCTEXT(\"[E5389DCD8C585B8C00B0C668CB1CF226]\", \"3615A19B4B210306EF4E2CA3E93B460D\", \"Poison Lobber\")"))
    (:NPC.CharacterType.Husk.RiotShield
      (return "NSLOCTEXT(\"[E5389DCD8C585B8C00B0C668CB1CF226]\", \"93C2C8D844AAA4333AD9D9BB7E7E4D6F\", \"Riot Husky\")"))
    (:NPC.CharacterType.Husk.Sploder
      (return "NSLOCTEXT(\"[E5389DCD8C585B8C00B0C668CB1CF226]\", \"6D7462854DFBBBB7B10A14B83D3F7620\", \"Sploder\")"))
    (:NPC.CharacterType.Husk.Zapper
      (return "NSLOCTEXT(\"[E5389DCD8C585B8C00B0C668CB1CF226]\", \"04AC184B4807EB5B830149B26F2BF967\", \"Gunslinger\")"))
    (:NPC.CharacterType.Shielder.Basic
      (return "NSLOCTEXT(\"[E5389DCD8C585B8C00B0C668CB1CF226]\", \"0DF7190C4CDEAC7083CDFA941D7BD8EE\", \"Shielder\")"))
    (:NPC.CharacterType.Shielder.Super
      (return "NSLOCTEXT(\"[E5389DCD8C585B8C00B0C668CB1CF226]\", \"9DB3C5CB41351C95F46368912E8EDC14\", \"Super Shielder\")"))
    (:NPC.CharacterType.Smasher.Basic
      (return "NSLOCTEXT(\"[E5389DCD8C585B8C00B0C668CB1CF226]\", \"081B8B224EB06471E1F37894EB5E687C\", \"Smasher\")"))
    (:NPC.CharacterType.Husk.Bombshell
      (return "NSLOCTEXT(\"[E5389DCD8C585B8C00B0C668CB1CF226]\", \"2DEF306D4E1884F3F17969A3181F2BA3\", \"Bombshell Lobber\")"))
    (:NPC.CharacterType.Taker.Basic
      (return "NSLOCTEXT(\"[E5389DCD8C585B8C00B0C668CB1CF226]\", \"AA0549C046236B279B28F7B71ED7D67B\", \"Taker\")"))))

(fn NPC_DroneStartRescue
  (params (Actor :type Actor%20Object%20Reference) (Instigator :type Actor%20Object%20Reference))
  (bind (_survivorhelpreminderduration_14_f54ddcc9447a56e42eea23b2f26a82bf _npc_ability_nonattack_cowerindefinitetrue_6_f917b44a44d629c4351341967fedce08 _npc_ability_nonattack_cowerindefinitefalse_7_e6f8c1c646c79def76809e829dca289d _npc_ability_nonattack_rescuetrue_9_6f9d8205427713c6b3c45d889df83be4 _npc_ability_nonattack_rescuefalse_11_b549d5ab4cbcbca2155c3faa098a58d0 _npc_ability_nonattack_lockinplacetrue_18_1fa6565d4bc9f11cb51ac584efde6999 _npc_ability_nonattack_lockinplacefalse_20_b578575c4157f18cad94eebb4ba45e71 _npc_ability_nonattack_rescuestart_28_f37eaff4488e12abfe11299818466044) (Utilities|Struct|BreakStructSurvivorScriptedAbilities (LocalVariables|GetSurvivorAbilityStruct)))
  (Ability|SendGameplayEventtoActor Actor _npc_ability_nonattack_rescuestart_28_f37eaff4488e12abfe11299818466044 (Utilities|Struct|MakeGameplayEventData _npc_ability_nonattack_rescuestart_28_f37eaff4488e12abfe11299818466044 Instigator)))

(fn NPC_DroneStopRescue
  (params (Actor :type Actor%20Object%20Reference) (Instigator :type Actor%20Object%20Reference))
  (bind (_survivorhelpreminderduration_14_f54ddcc9447a56e42eea23b2f26a82bf _npc_ability_nonattack_cowerindefinitetrue_6_f917b44a44d629c4351341967fedce08 _npc_ability_nonattack_cowerindefinitefalse_7_e6f8c1c646c79def76809e829dca289d _npc_ability_nonattack_rescuetrue_9_6f9d8205427713c6b3c45d889df83be4 _npc_ability_nonattack_rescuefalse_11_b549d5ab4cbcbca2155c3faa098a58d0 _npc_ability_nonattack_lockinplacetrue_18_1fa6565d4bc9f11cb51ac584efde6999 _npc_ability_nonattack_lockinplacefalse_20_b578575c4157f18cad94eebb4ba45e71 _npc_ability_nonattack_rescuestart_28_f37eaff4488e12abfe11299818466044 _npc_ability_nonattack_rescuefinish_29_7b19bd2348aa524234d34493c53321f3) (Utilities|Struct|BreakStructSurvivorScriptedAbilities (LocalVariables|GetSurvivorAbilityStruct)))
  (Ability|SendGameplayEventtoActor Actor _npc_ability_nonattack_rescuefinish_29_7b19bd2348aa524234d34493c53321f3 (Utilities|Struct|MakeGameplayEventData _npc_ability_nonattack_rescuefinish_29_7b19bd2348aa524234d34493c53321f3 Instigator)))

(fn NPC_LockThenRotateInPlace
  (params (LockedInPlace :type Boolean) (PlayerPawn :type Fort%20Player%20Pawn%20Object%20Reference) (SurvivorPawn :type Fort%20AIPawn%20Object%20Reference))
  (bind _survivor_abilities (LocalVariables|GetSurvivor%20Abilities))
  (bind (_survivorhelpreminderduration_14_f54ddcc9447a56e42eea23b2f26a82bf _npc_ability_nonattack_cowerindefinitetrue_6_f917b44a44d629c4351341967fedce08 _npc_ability_nonattack_cowerindefinitefalse_7_e6f8c1c646c79def76809e829dca289d _npc_ability_nonattack_rescuetrue_9_6f9d8205427713c6b3c45d889df83be4 _npc_ability_nonattack_rescuefalse_11_b549d5ab4cbcbca2155c3faa098a58d0 _npc_ability_nonattack_lockinplacetrue_18_1fa6565d4bc9f11cb51ac584efde6999 _npc_ability_nonattack_lockinplacefalse_20_b578575c4157f18cad94eebb4ba45e71) (Utilities|Struct|BreakStructSurvivorScriptedAbilities _survivor_abilities))
  (if LockedInPlace
    (Ability|SendGameplayEventtoActor SurvivorPawn _npc_ability_nonattack_lockinplacetrue_18_1fa6565d4bc9f11cb51ac584efde6999 (Utilities|Struct|MakeGameplayEventData _npc_ability_nonattack_lockinplacetrue_18_1fa6565d4bc9f11cb51ac584efde6999 PlayerPawn))
    (else
      (Ability|SendGameplayEventtoActor SurvivorPawn _npc_ability_nonattack_lockinplacefalse_20_b578575c4157f18cad94eebb4ba45e71 (Utilities|Struct|MakeGameplayEventData _npc_ability_nonattack_lockinplacefalse_20_b578575c4157f18cad94eebb4ba45e71 SurvivorPawn)))))

(fn PointLocationsBetweenTwoVectors
  (params (Vector%201 :type Vector) (Vector%202 :type Vector) (HowManyPoints :type Integer :default 2))
  (if (>= HowManyPoints 2)
    (LocalVariables|SetDifferenceVector (/ (- Vector%202 Vector%201) HowManyPoints))
    (for _index (range 1 (+ HowManyPoints 1))
      (Utilities|Array|AddUnique (LocalVariables|GetResultVectorArray) (+ (* (LocalVariables|GetDifferenceVector) _index) Vector%201)))
    (if (>= (Utilities|Array|Length (LocalVariables|GetResultVectorArray)) 1)
      (LocalVariables|SetWasSuccessful true)
      (else
        (LocalVariables|SetWasSuccessful false)))
    (else
      (LocalVariables|SetWasSuccessful false)))
  (return (LocalVariables|GetResultVectorArray) (LocalVariables|GetWasSuccessful)))

(fn QuestObjectiveCountAchieved
  (params (PlayerControllerReference :type Fort%20Player%20Controller%20Object%20Reference) (Quest%20Item%20Reference :type Fort%20Quest%20Item%20Definition%20Object%20Reference) (ObjectiveStatHandle :type Data%20Table%20Row%20Handle%20Structure))
  (bind _returnvalue (Quests|GetQuestManager PlayerControllerReference "Campaign"))
  (bind _returnvalue_2 (Quests|GetQuestwithDefinition _returnvalue Quest%20Item%20Reference))
  (bind _returnvalue_3 (Quest|GetObjectiveInfo _returnvalue_2 ObjectiveStatHandle))
  (Utilities|IsValid PlayerControllerReference
    (:Is%20Valid
      (if (Utilities|IsValid _returnvalue_2 :!class /Script/Engine.KismetSystemLibrary)
        (LocalVariables|SetbSuccessfullyFoundStage true)
        (LocalVariables|SetLocalAchievedCount (Variables|/Script/FortniteGame.FortQuestObjectiveInfo|GetAchievedCount _returnvalue_3))
        (LocalVariables|SetLocalRequiredCount (Variables|/Script/FortniteGame.FortQuestObjectiveInfo|GetRequiredCount _returnvalue_3))
        (else
          (LocalVariables|SetbSuccessfullyFoundStage false :!label _label)))
      (return (LocalVariables|GetLocalAchievedCount) (LocalVariables|GetLocalRequiredCount) (LocalVariables|GetbSuccessfullyFoundStage)))
    (:Is%20Not%20Valid
      _))
  (goto _label))

(fn Random_360_VectorInRange
  (params (MinDistance :type Float%20%28double-precision%29) (MaxDistance :type Float%20%28double-precision%29))
  (return (Math|Vector|RotateVectorAroundAxis (Math|Vector|MakeVector (Math|Random|RandomFloatinRange MinDistance MaxDistance)) (Math|Random|RandomFloatinRange 0.0 360) "0, 0,1.000000")))

(fn RegisterUIUpdatesWithDamageEvents
  (params (ActorToRegisterWith :type Actor%20Object%20Reference) (MissionState :type Fort%20Mission%20State%20Object%20Reference))
  (bind _asbuilding_actor (Utilities|Casting|CastToBuildingActor ActorToRegisterWith)
    (:then
      (Health|BindEventtoOnBuildingHealthChanged _asbuilding_actor (EventDispatchers|CreateEvent MissionState :!event UpdateUI)))
    (:CastFailed
      (bind _asfort_pawn (Utilities|Casting|CastToFortPawn ActorToRegisterWith))
      (Health|BindEventtoOnPawnHealthChanged _asfort_pawn (EventDispatchers|CreateEvent MissionState :!event UpdateUI)))))

(fn SeperatePlayersWhoNeedQuest
  (params (QuestItemReference :type Fort%20Quest%20Item%20Definition%20Object%20Reference) (QuestObjectiveBackendName :type Name))
  (for _array_element (Fortnite|GetAllFortPlayerControllers true)
    (Utilities|IsValid _array_element
      (:Is%20Valid
        (if (Utilities|IsValid (Quests|GetQuestwithDefinition (Quests|GetQuestManager _array_element "Invalid") QuestItemReference) :!class /Script/Engine.KismetSystemLibrary)
          (if (Quest|HasCompletedObjectivewithName (Quests|GetQuestwithDefinition (Quests|GetQuestManager _array_element "Invalid") QuestItemReference) QuestObjectiveBackendName :!class /Script/FortniteGame.FortQuestItem)
            (Utilities|Array|AddUnique (LocalVariables|GetNotRequireQuest) _array_element)
            (else
              (Utilities|Array|AddUnique (LocalVariables|GetRequireQuest) _array_element)))
          (else
            (Utilities|Array|AddUnique (LocalVariables|GetNotRequireQuest) _array_element))))))
  (return (LocalVariables|GetRequireQuest) (LocalVariables|GetNotRequireQuest)))

(fn UnRegisterUIUpdatesWithDamageEvents
  (params (ActorToUnRegisterWith :type Actor%20Object%20Reference) (MissionState :type Fort%20Mission%20State%20Object%20Reference))
  (bind _asbuilding_actor (Utilities|Casting|CastToBuildingActor ActorToUnRegisterWith)
    (:then
      (Health|UnbindEventfromOnBuildingHealthChanged _asbuilding_actor (EventDispatchers|CreateEvent MissionState :!event UpdateUI)))
    (:CastFailed
      (bind _asfort_pawn (Utilities|Casting|CastToFortPawn ActorToUnRegisterWith))
      (Health|UnbindEventfromOnPawnHealthChanged _asfort_pawn (EventDispatchers|CreateEvent MissionState :!event UpdateUI)))))
